EditorUtilityWidgetBlueprint /MusicEncounter/Tools/MusicEncounterNPCWorkflow/EUW_MusicEncounter_NPC_Creator
graphs [
 EventGraph {nodes [
   "On Clicked (CreateAssetsButton)"() {is=title K2Node_ComponentBoundEvent}
   > RunCreate()
   Event PreConstruct()
   > SetObject(self=Get CustomPartDetailView() -> (CustomPartDetailView:Details View Object Reference), NewObject=Self-Reference())
   > SetObject(self=Get CustomPartDetailView_1() -> (CustomPartDetailView_1:Details View Object Reference), NewObject=Self-Reference())
  ]}
 RunCreate {nodes [
   RunCreate() {is=title K2Node_FunctionEntry}
   > n1 = CreateSourcePresetBuilder!(
    execute
    self:Meta Sound Builder Subsystem Object Reference=Get MetaSoundBuilderSubsystem() -> (ReturnValue:Meta Sound Builder Subsystem Object Reference) {is=title K2Node_GetEngineSubsystem}
    BuilderName:Name=MusicEncounter
    ReferencedSourceClass:"Meta Sound Document Interface Interface (by ref)"=Get EncounterPreset() -> (EncounterPreset:Meta Sound Document Interface Interface)
   ) -> (Succeeded:Exec, Failed:Exec, ReturnValue:Meta Sound Source Builder Object Reference)
   > Set BuilderRef(BuilderRef:Meta Sound Source Builder Object Reference=@n1) -> (Output_Get:Meta Sound Source Builder Object Reference)
   > n2 = Sequence()
   > n3 = Create MetaSound Literal From AudioParameter!(
    execute
    self:Meta Sound Builder Subsystem Object Reference=Get MetaSoundBuilderSubsystem()
    Param:"Audio Parameter Structure (by ref)"=ObjectToAudioParameter!(Name:Name=Idle.Wave, Object:Object Reference=Get Idle() -> (Idle:Sound Wave Object Reference)) -> (ReturnValue:Audio Parameter Structure)
   ) -> (then, ReturnValue:Meta Sound Literal Structure)
   > SetGraphInputDefault!(
    execute
    self:Meta Sound Builder Base Object Reference=Get BuilderRef() -> (BuilderRef:Meta Sound Source Builder Object Reference)
    InputName:Name=Idle.Wave
    Literal:"Meta Sound Literal Structure (by ref)"=@n3
    PageName:Name
   ) -> (Succeeded:Exec, Failed:Exec)
   > n4 = Create MetaSound Literal From AudioParameter(
    self=Get MetaSoundBuilderSubsystem()
    Param=ObjectToAudioParameter(Name=Combat.TransitionWave, Object=Get CombatTransition() -> (CombatTransition:Sound Wave Object Reference))
   )
   > SetGraphInputDefault(self=Get BuilderRef(), InputName=Combat.TransitionWave, Literal=@n4)
   > n5 = Create MetaSound Literal From AudioParameter(
    self=Get MetaSoundBuilderSubsystem()
    Param=ObjectToAudioParameter(Name=Combat.PercWave, Object=Get Combat() -> (Combat:Sound Wave Object Reference))
   )
   > SetGraphInputDefault(self=Get BuilderRef(), InputName=Combat.PercWave, Literal=@n5)
   > n6 = Create MetaSound Literal From AudioParameter(
    self=Get MetaSoundBuilderSubsystem()
    Param=ObjectToAudioParameter(Name=Defeated.Wave, Object=Get Defeated() -> (Defeated:Sound Wave Object Reference))
   )
   > SetGraphInputDefault(self=Get BuilderRef(), InputName=Defeated.Wave, Literal=@n6)
   n7 = Sequence(execute=@n2.then_1)
   > n8 = Create MetaSound Literal From AudioParameter(
    self=Get MetaSoundBuilderSubsystem()
    Param=IntegerToAudioParameter!(Name:Name=Combat.BeatOffset, Integer:int=Get CombatTransitionOffset() -> (CombatTransitionOffset:int=0)) -> (ReturnValue:Audio Parameter Structure)
   )
   > SetGraphInputDefault(self=Get BuilderRef(), InputName=Combat.BeatOffset, Literal=@n8)
   n9 = Get Tempo() -> (Tempo:int=0)
   > n10 = Create MetaSound Literal From AudioParameter(
    self=Get MetaSoundBuilderSubsystem()
    Param=FloatToAudioParameter!(Name:Name=MusicEncounter.Tempo, Float:float="To Float (Integer)"(InInt=@n9)) -> (ReturnValue:Audio Parameter Structure)
   )
   > SetGraphInputDefault(self=Get BuilderRef(), InputName=MusicEncounter.Tempo, Literal=@n10)
   n11 = Get EUW_MetaSoundAssetPath() -> (EUW_MetaSoundAssetPath:EUW Meta Sound Asset Path Weapon Object Reference)
   n12 = BuildToAsset!(
    execute=@n7.then_1
    self:Meta Sound Editor Subsystem Object Reference=Get MetaSoundEditorSubsystem() -> (ReturnValue:Meta Sound Editor Subsystem Object Reference) {is=title K2Node_GetEditorSubsystem}
    InBuilder:Meta Sound Builder Base Object Reference=Get BuilderRef()
    Author:String=SoundBaseAssetAction
    AssetName:String="To String (Text)"(InText=GetAssetName!(self:EUW Meta Sound Asset Path Weapon Object Reference=@n11) -> (AssetText:Text, IsEmpty:bool))
    PackagePath:String="To String (Text)"(InText=GetPath!(self:EUW Meta Sound Asset Path Weapon Object Reference=@n11) -> (PathText:Text, bIsEmpty:bool))
    TemplateSoundWave:Sound Wave Object Reference
   ) -> (Succeeded:Exec, Failed:Exec, ReturnValue:Meta Sound Document Interface Interface)
   > n13 = Cast To MetaSoundSource!(execute, Object:Meta Sound Document Interface Interface=@n12) -> (then, CastFailed, AsMetaSound Source:Meta Sound Source Object Reference) {is=title K2Node_DynamicCast}
   > n14 = GetAssetTools()
   n15 = GetAssetName(self=Get EUW_MetaSoundAssetPath())
   n16 = "To Text (String)"(InString=Replace(SourceString="To String (Text)"(InText=@n15), From=MSS_))
   n17 = Get EUW_MetaSoundAssetPath()
   n18 = MakeSoftClassPath(PathString=/Script/Engine.Blueprint)
   > n19 = CreateAsset(
    self=@n14
    AssetName="To String (Text)"(InText=Format Text!(Format="LOCTEXT(\"BP_MusicEncounter_{Name}\")", Name:Text=@n16) -> (Result))
    PackagePath="To String (Text)"(InText=GetPath(self=@n17))
    AssetClass=Resolve Soft Reference!(Input:Object Soft Class Reference=To Soft Class Reference(SoftClassPath=@n18)) -> (Output:Object Class Reference)
   )
   > n20 = GetBlueprintAsset(Object=@n19)
   > ReparentBlueprint(Blueprint=@n20, NewParentClass=/MusicEncounter/Encounters/BP_MusicEncounter_NPC_Boss.BP_MusicEncounter_NPC_Boss_C)
   > CompileBlueprint!(execute, Blueprint:Blueprint Object Reference=@n20) -> (then, ReturnValue:bool)
   > n21 = GeneratedClass!(execute, BlueprintObj:Blueprint Object Reference=@n20) -> (then, ReturnValue:Object Class Reference)
   > n22 = GetClassDefaultObject!(execute, Class:Object Class Reference=@n21) -> (then, ReturnValue:Object Reference)
   > n23 = Cast To BP_MusicEncounter_NPC_Boss!(execute, Object=@n22) -> (then, CastFailed, AsBP Music Encounter NPC Boss:BP Music Encounter NPC Boss Object Reference) {is=title K2Node_DynamicCast}
   > Set NPC Music!(execute, NPC Music:Meta Sound Source Object Reference=@n13, self:BP Music Encounter NPC Boss Object Reference=@n23) -> (then, Output_Get:Meta Sound Source Object Reference)
   > n24 = GetBlueprintAsset(Object=Get ActorTarget() -> (ActorTarget:Object Reference))
   > n25 = GeneratedClass(BlueprintObj=@n24)
   > n26 = GetClassDefaultObject(Class=@n25)
   > n27 = Cast To BP_PlayerPawn_NonParticipant!(execute, Object=@n26) -> (then, CastFailed, AsBP Player Pawn Non Participant:BP Player Pawn Non Participant Object Reference) {is=title K2Node_DynamicCast}
   > n28 = Gather Subobject Data For Instance(self=Get SubobjectDataSubsystem(), Context=@n27)
   > n29 = For Each Loop(Array:Array of Subobject Data Handle Structures=@n28) -> (Array Element:Subobject Data Handle Structure)
   n30 = "Get (a ref)"(Array:Array of Subobject Data Handle Structures=@n28, Dimension 1=0) -> (Output:"Subobject Data Handle Structure (by ref)")
   n31 = AddNewSubobject!(
    execute=@n29.Completed
    self:Subobject Data Subsystem Object Reference=Get SubobjectDataSubsystem()
    Params_ParentHandle:Subobject Data Handle Structure=@n30
    Params_NewClass:Object Class Reference=/MusicEncounter/Components/BP_MusicEncounterComponent_NPC_Boss.BP_MusicEncounterComponent_NPC_Boss_C
    Params_BlueprintContext:Blueprint Object Reference=GetBlueprint!(Data:"Subobject Data Structure (by ref)"=GetData(DataHandle=@n30)) -> (ReturnValue:Blueprint Object Reference)
    Params_bSkipMarkBlueprintModified:bool=False
    Params_bConformTransformToParent:bool=false
   ) -> (then, FailReason:Text, ReturnValue:Subobject Data Handle Structure)
   n32 = Get SubobjectDataSubsystem()
   n33 = Make Soft Reference!(Input:Object Class Reference=@n21) -> (Output:Object Soft Class Reference)
   > SetEditorProperty(
    Object=GetAssociatedObject!(Data:"Subobject Data Structure (by ref)"=FindSubobjectDataFromHandle!(self:Subobject Data Subsystem Object Reference=@n32, Handle:"Subobject Data Handle Structure (by ref)"=@n31.ReturnValue) -> (OutData:Subobject Data Structure, ReturnValue:bool)) -> (ReturnValue:Object Reference)
    PropertyName=Music Encounter Class
    PropertyValue:Object Soft Class Reference=@n33
   )
   n34 = GetVariableName!(Data:"Subobject Data Structure (by ref)"=GetData(DataHandle=@n29)) -> (ReturnValue:Name=None)
   Branch(execute=@n29.LoopBody, Condition=Contains(SearchIn="To String (Name)"(InName=@n34), Substring=MusicEncounterComponent))
   n35 = Get SubobjectDataSubsystem()
   > SetEditorProperty(Object=GetAssociatedObject(Data=FindSubobjectDataFromHandle(self=@n35, Handle=@n29)), PropertyName=Music Encounter Class, PropertyValue=@n33)
   > Return Node()
  ]}
]
variables [
 ActorTarget {type=Object, default=null}
 BuilderRef {type=MetaSoundSourceBuilder, default=null}
 Combat {type=SoundWave, default=null}
 CombatTransition {type=SoundWave, default=null}
 CombatTransitionOffset {type=integer, default=1}
 Defeated {type=SoundWave, default=null}
 EncounterPreset {default=MetaSoundSource'/MusicEncounter/MetaSounds/MSS_MusicEncounter_NPC_Default'}
 Idle {type=SoundWave, default=null}
 MusicEncounterComponents {type=Enum_MusicEncounterComponents, default=Music Encounter Component NPC}
 MusicEncounters {type=Enum_MusicEncounters, default=Music Encounter NPC}
 Tempo {type=integer, default=120}
]
widgets {
 info {rootWidgetClass=&CanvasPanel, widgetCount=11}
 widgets [
  CanvasPanel_21 {widgetClassPath=&CanvasPanel, widget=CanvasPanel_21}
  EUW_MetaSoundAssetPath {
   widgetClassPath=/Game/Sounds/Fort_Audio_Blueprints/Utilities/EditorUtilityWidgets/Metasounds/Weapons/EUW_MetaSoundAssetPath_Weapon
   bIsVariable=true
   parent=CanvasPanel_21
   widget=EUW_MetaSoundAssetPath
   slot {class=CanvasPanelSlot, properties=~}
  }
  VerticalBox_29 {
   widgetClassPath=&VerticalBox
   parent=CanvasPanel_21
   widget=VerticalBox_29
   slot {
    class=CanvasPanelSlot
    properties {layoutData {alignment=<0 0>, anchors {maximum=<0 0>, minimum=<0 0>}, offsets=<ltrb 4 136 481.0128 185.0564>}}
   }
  }
  SizeBox_72 {
   widgetClassPath=&SizeBox
   parent=VerticalBox_29
   widget=SizeBox_72
   slot {class=VerticalBoxSlot, properties {size {sizeRule=Fill, value=1}}}
  }
  CustomPartDetailView {
   widgetClassPath=&DetailsView
   parent=SizeBox_72
   widget=CustomPartDetailView
   properties {columnWidth=0.5, categoriesToShow [Actor, Music Encounter Config]}
   slot {class=SizeBoxSlot}
  }
  VerticalBox_23 {
   widgetClassPath=&VerticalBox
   parent=CanvasPanel_21
   widget=VerticalBox_23
   slot {
    class=CanvasPanelSlot
    properties {layoutData {alignment=<0 0>, anchors {maximum=<0 0>, minimum=<0 0>}, offsets=<ltrb 0 292 480.3804 314.4845>}}
   }
  }
  CustomPartDetailView_1 {
   widgetClassPath=&DetailsView
   parent=VerticalBox_23
   widget=CustomPartDetailView_1
   properties {columnWidth=0.5, categoriesToShow [Music Encounter NPC]}
   slot {class=VerticalBoxSlot}
  }
  CreateAssetsButtonBox {
   widgetClassPath=&HorizontalBox
   parent=CanvasPanel_21
   widget=CreateAssetsButtonBox
   slot {class=CanvasPanelSlot, properties {layoutData {alignment=<0 0>, anchors {maximum=<0 0>, minimum=<0 0>}, offsets=<ltrb 132 628 216.48 30>}}}
  }
  WrapBox_2 {
   widgetClassPath=&WrapBox
   parent=CreateAssetsButtonBox
   widget=WrapBox_2
   properties {horizontalAlignment=HAlign_Fill, innerSlotPadding=<4 4>}
   slot {class=HorizontalBoxSlot, properties {size {sizeRule=Fill, value=1}}}
  }
  CreateAssetsButton {
   widgetClassPath=&EditorUtilityButton
   parent=WrapBox_2
   widget=CreateAssetsButton
   properties {widgetStyle {
     clickedSlateSound {resourceObject=null}
     disabled {
      bIsDynamicallyLoaded=false
      drawAs=RoundedBox
      imageType=NoImage
      mirroring=NoMirror
      resourceName=None
      resourceObject=null
      tiling=NoTile
      imageSize=<0 0>
      margin=<ltrb 0 0 0 0>
      outlineSettings {
       bUseBrushTransparency=false
       roundingType=FixedRadius
       width=1
       color {colorUseRule=UseColor_ColorTable, specifiedColor=<rgba 0 0 0 0>}
       cornerRadii=<4 4 4 4>
      }
      tintColor {colorUseRule=UseColor_Specified, specifiedColor=<rgba 0.0395 0.0395 0.0395 1>}
      uVRegion {bIsValid=false, max=<0 0>, min=<0 0>}
     }
     disabledForeground {colorUseRule=UseColor_ColorTable, specifiedColor=<rgba 0 0 0 0>}
     hovered {
      bIsDynamicallyLoaded=false
      drawAs=RoundedBox
      imageType=NoImage
      mirroring=NoMirror
      resourceName=None
      resourceObject=null
      tiling=NoTile
      imageSize=<32 32>
      margin=<ltrb 0 0 0 0>
      outlineSettings {
       bUseBrushTransparency=false
       roundingType=FixedRadius
       width=1
       color {colorUseRule=UseColor_ColorTable, specifiedColor=<rgba 0 0 0 0>}
       cornerRadii=<4 4 4 4>
      }
      tintColor {colorUseRule=UseColor_Specified, specifiedColor=<rgba 0.0044 0.2384 1 1>}
      uVRegion {bIsValid=false, max=<0 0>, min=<0 0>}
     }
     hoveredForeground {colorUseRule=UseColor_ColorTable, specifiedColor=<rgba 0 0 0 0>}
     hoveredSlateSound {resourceObject=null}
     normal {
      bIsDynamicallyLoaded=false
      drawAs=RoundedBox
      imageType=NoImage
      mirroring=NoMirror
      resourceName=None
      resourceObject=null
      tiling=NoTile
      imageSize=<32 32>
      margin=<ltrb 0 0 0 0>
      outlineSettings {
       bUseBrushTransparency=false
       roundingType=FixedRadius
       width=1
       color {colorUseRule=UseColor_ColorTable, specifiedColor=<rgba 0 0 0 0>}
       cornerRadii=<4 4 4 4>
      }
      tintColor {colorUseRule=UseColor_Specified, specifiedColor=<rgba 0 0.162 0.7454 1>}
      uVRegion {bIsValid=false, max=<0 0>, min=<0 0>}
     }
     normalForeground {colorUseRule=UseColor_ColorTable, specifiedColor=<rgba 0 0 0 0>}
     normalPadding=<ltrb 12 1.5 12 1.5>
     pressed {
      bIsDynamicallyLoaded=false
      drawAs=RoundedBox
      imageType=NoImage
      mirroring=NoMirror
      resourceName=None
      resourceObject=null
      tiling=NoTile
      imageSize=<32 32>
      margin=<ltrb 0 0 0 0>
      outlineSettings {
       bUseBrushTransparency=false
       roundingType=FixedRadius
       width=1
       color {colorUseRule=UseColor_ColorTable, specifiedColor=<rgba 0 0 0 0>}
       cornerRadii=<4 4 4 4>
      }
      tintColor {colorUseRule=UseColor_Specified, specifiedColor=<rgba 0 0.0802 0.3515 1>}
      uVRegion {bIsValid=false, max=<0 0>, min=<0 0>}
     }
     pressedForeground {colorUseRule=UseColor_ColorTable, specifiedColor=<rgba 0 0 0 0>}
     pressedPadding=<ltrb 12 2.5 12 0.5>
     pressedSlateSound {resourceObject=null}
    }}
   slot {class=WrapBoxSlot}
  }
  TextBlock_295 {
   widgetClassPath=&TextBlock
   parent=CreateAssetsButton
   widget=TextBlock_295
   properties {
    text=Create Assets
    font {
     bForceMonospaced=false
     bMaterialIsStencil=false
     fontMaterial=null
     fontObject=&Burbank
     letterSpacing=0
     monospacedWidth=1
     size=12
     skewAmount=0
     typefaceFontName=Burbank-Black
     outlineSettings {
      bApplyOutlineToDropShadows=false
      bMiteredCorners=false
      bSeparateFillAlpha=false
      outlineMaterial=null
      outlineSize=0
      outlineColor=<rgba 0 0 0 1>
     }
    }
    margin=<ltrb 0 1 0 0>
   }
   slot {class=ButtonSlot}
  }
 ]
}
